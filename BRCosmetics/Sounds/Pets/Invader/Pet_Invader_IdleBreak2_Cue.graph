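# Sound cue for the second idle-break sound of the Invader pet cosmetic.
# A Random node picks one of three wave variations with equal weight.
# It plays through the creature pets sound class with the close-range pet attenuation asset and a slightly reduced volume.
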SoundCue /BRCosmetics/Sounds/Pets/Invader/Pet_Invader_IdleBreak2_Cue
root=Random_0
nodes [
 {id=Random_0, children [WavePlayer_0,WavePlayer_1,WavePlayer_2], properties {weights [1,1,1]}}
 {id=WavePlayer_0, properties {soundWaveAssetPtr=&/Pet_Invader_IdleBreak2_01}}
 {id=WavePlayer_1, properties {soundWaveAssetPtr=&/Pet_Invader_IdleBreak2_02}}
 {id=WavePlayer_2, properties {soundWaveAssetPtr=&/Pet_Invader_IdleBreak2_03}}
]
properties {
 attenuationSettings=&Fort_Pets_Close_Athena_Att
 maxDistance=1800.0001
 soundClassObject=&Fort_Creatures_Pets
 volumeMultiplier=0.896
}
